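; AM_Delete_Bone_Data is an animation modifier that strips animation data for a chosen bone from the sequence it is applied to.
; On apply, it calls RemoveBoneAnimation from the Animation Blueprint Library, passing the Bone Name and Include Children variables, so the bone's descendants can also be removed.
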(event EventOnApply
  (params (AnimationSequence :type Anim%20Sequence%20Object%20Reference))
  (AnimationBlueprintLibrary|Bones|RemoveBoneAnimation AnimationSequence (Variables|Default|GetBone%20Name) (Variables|Default|GetInclude%20Children)))
