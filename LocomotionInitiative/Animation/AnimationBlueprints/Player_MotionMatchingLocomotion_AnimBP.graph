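# Player locomotion Animation Blueprint for the Fortnite character skeleton, built on motion matching with orientation warping.
# Motion matching gets pelvis adjustment and chooser-driven leans, then feeds an Idle/Moving machine with idle breaks, turn-in-place shuffles and blendspace alternatives.
# A CrouchTransition machine layers additive crouch transitions over that pose, and Pose History records the output with trajectory.
# A low-LOD blendspace state machine exists but is unconnected to Output Pose.
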AnimBlueprint /LocomotionInitiative/Animation/AnimationBlueprints/Player_MotionMatchingLocomotion_AnimBP
target_skeleton=&Fortnite_M_Avg_Player_Skeleton
graphs [
 AnimGraph {nodes [
   AnimGraphNode_StateMachine_2 = LowLODMovement() {is=title AnimGraphNode_StateMachine, bAllowConduitEntryStates=true}
   Blend Poses by bool(BlendPose_0=@AnimGraphNode_StateMachine_2, bActiveValue=Get bUseLowLODLocomotion() -> (bUseLowLODLocomotion:bool)) {blendType=Linear}
   AnimGraphNode_StateMachine_0 = CrouchTransition() {is=title AnimGraphNode_StateMachine, bAllowConduitEntryStates=true}
   AnimGraphNode_PoseSearchHistoryCollector_0 = Pose History!(
    Source:Pose Link Structure=@AnimGraphNode_StateMachine_0
    TransformTrajectory:Transform Trajectory Structure=Get Trajectory() -> (Trajectory:Transform Trajectory Structure)
    TrajectorySpeedMultiplier:float=Get TrajectorySpeedMultiplier() -> (TrajectorySpeedMultiplier:float)
   ) -> (Pose:Pose Link Structure) {
    is=title AnimGraphNode_PoseSearchHistoryCollector
    binding=&self:AnimGraph.AnimGraphNode_PoseSearchHistoryCollector_0.Binding
    poseCount=3
    collectedBones [{boneName=foot_l}, {boneName=foot_r}, {boneName=thigh_l}, {boneName=thigh_r}]
   }
   Output Pose(Result=@AnimGraphNode_PoseSearchHistoryCollector_0) {layerGroup=None}
   AnimGraphNode_StateMachine_1 = Locomotion() {is=title AnimGraphNode_StateMachine, bAllowConduitEntryStates=true, maxTransitionsPerFrame=1}
   Save cached pose 'LocomotionPose'(Pose=@AnimGraphNode_StateMachine_1) {cacheName=LocomotionPose}
   AnimGraphNode_ChooserPlayer_2 = Chooser Player: CT_Add_Leans!(BlendSpaceX:float=0) -> (Pose:Pose Link Structure) {
    is=title AnimGraphNode_ChooserPlayer
    evaluationFrequency=OnUpdate
    chooser {_structType=/Script/Chooser.EvaluateChooser, chooser=/LocomotionInitiative/Animation/Locomotion_anims/MotionMatching/CT_Add_Leans}
    chooserContextDefinition [
     {
      _structType=/Script/Chooser.ContextObjectTypeClass
      class=&/Player_MotionMatchingLocomotion_AnimBP.Player_MotionMatchingLocomotion_AnimBP_C
      direction=ReadWrite
     }
     {struct=/Script/Chooser.ChooserPlayerSettings, _structType=/Script/Chooser.ContextObjectTypeStruct, direction=Read}
    ]
    defaultSettings {blendOption=Linear}
   }
   n1 = Apply Mesh Space Additive!(
    Base:Pose Link Structure=Use cached pose 'MM Result'()
    Additive:Pose Link Structure=@AnimGraphNode_ChooserPlayer_2
    Alpha:float=Get AdditiveLeanAlpha() -> (AdditiveLeanAlpha:float)
   ) -> (Pose:Pose Link Structure) {
    is=title AnimGraphNode_ApplyMeshSpaceAdditive
    alphaBoolBlend {blendOption=Linear}
    alphaScaleBiasClamp {bInterpResult=true, interpSpeedDecreasing=5, interpSpeedIncreasing=5}
   }
   Save cached pose 'LocomotionBase'(Pose=@n1) {cacheName=LocomotionBase}
   n2 = Local To Component(LocalPose=Use cached pose 'PrePelvisRotation'())
   n3 = Get PelvisRotationAdjustment() -> (PelvisRotationAdjustment:Rotator)
   n4 = Get!() -> (PelvisRotationAdjustment_Roll:float, PelvisRotationAdjustment_Pitch:float, PelvisRotationAdjustment_Yaw:float)
   n5 = "Transform (Modify) Bone - Bone: ik_foot_r"!(
    ComponentPose="Transform (Modify) Bone - Bone: ik_foot_l"!(
     ComponentPose="Transform (Modify) Bone - Bone: pelvis"!(ComponentPose=@n2, Alpha=1, Rotation=@n3) -> (Pose) {rotationMode=BMM_Additive, alphaBoolBlend {blendOption=Linear}, boneToModify {boneName=pelvis}}
     Alpha=1
     Rotation_Roll:float=0
     Rotation_Pitch:float=0
     Rotation_Yaw:float=@n4.PelvisRotationAdjustment_Yaw
    ) -> (Pose) {rotationMode=BMM_Additive, alphaBoolBlend {blendOption=Linear}, boneToModify {boneName=ik_foot_l}}
    Alpha=1
    Rotation_Roll:float=0
    Rotation_Pitch:float=0
    Rotation_Yaw:float=@n4.PelvisRotationAdjustment_Yaw
   ) -> (Pose) {rotationMode=BMM_Additive, alphaBoolBlend {blendOption=Linear}, boneToModify {boneName=ik_foot_r}}
   Save cached pose 'MM Result'(Pose=Blend Poses by bool(
     BlendPose_0=Use cached pose 'PrePelvisRotation'()
     BlendPose_1=Component To Local(ComponentPose=@n5)
     BlendTime_0=0
     BlendTime_1=0
     bActiveValue=Get bIsCrouching()
    )) {cacheName=MM Result}
   AnimGraphNode_MotionMatching_1 = Motion Matching!(
    BlendTime:float=Break Motion Matching Blend Info!(MotionMatchingBlendInfo:"Motion Matching Blend Info Structure (by ref)"=Get MotionMatchingBlendInfo() -> (MotionMatchingBlendInfo:Motion Matching Blend Info Structure)) -> (BlendTime:float=~, BlendProfile:Name) {is=title K2Node_BreakStruct}
    BlendProfile:Blend Profile Object Reference
    SearchThrottleTime:float=Get SearchThrottleTime() -> (SearchThrottleTime:float)
    PlayRateMultiplier:float=Get PlayRateMultiplier() -> (PlayRateMultiplier:float)
    bShouldSearch:bool=Get bShouldExecuteMotionMatchingSearch() -> (bShouldExecuteMotionMatchingSearch:bool)
   ) -> (Pose:Pose Link Structure) {
    is=title AnimGraphNode_MotionMatching
    node {
     bShouldFilterNotifies=true
     bStoreBlendedPose=false
     blendOption=HermiteCubic
     blendTime=0.37
     maxActiveBlends=3
     maxBlendInTimeToOverrideAnimation=0.05
     playerDepthBlendInTimeMultiplier=1.3
     onMotionMatchingStateUpdated {functionName=OnUpdateMotionMatchingState}
     playRate {max=1.05, min=0.95}
     poseJumpThresholdTime {max=0.3, min=-0.3}
     updateFunction {functionName=OnUpdateMotionMatching}
    }
    onMotionMatchingStateUpdatedFunction {bSelfContext=true, memberName=OnUpdateMotionMatchingState}
    updateFunction {bSelfContext=true, memberName=OnUpdateMotionMatching}
   }
   Save cached pose 'PrePelvisRotation'(Pose=@AnimGraphNode_MotionMatching_1) {cacheName=PrePelvisRotation}
  ]}
 AnimationBlendStackGraph_0 {
  path=AnimGraph.AnimGraphNode_ChooserPlayer_2.AnimationBlendStackGraph_0
  nodes [Output Pose!(Result:Pose Link Structure=Blend Stack Input!() -> (Pose:Pose Link Structure) {is=title AnimGraphNode_BlendStackInput}) {is=title AnimGraphNode_BlendStackResult, layerGroup=None}]
 }
 AnimationBlendStackGraph_0 {
  path=AnimGraph.AnimGraphNode_MotionMatching_1.AnimationBlendStackGraph_0
  nodes [
   n1 = Orientation Warping!(
    ComponentPose:Component Space Pose Link Structure=Local To Component(LocalPose=Blend Stack Input())
    Alpha:float=Get OrientationWarpingAlpha() -> (OrientationWarpingAlpha:float)
    TargetTime:float=0.8
    LocomotionAngle:float=0
    LocomotionDirection:Vector="0,0,0"
    CurrentAnimAsset:Animation Asset Object Reference
    CurrentAnimAssetTime:float=0
    bUseManualRootMotionVelocity:bool=True
    ManualRootMotionVelocity:Vector="0,0,0"
   ) -> (Pose:Component Space Pose Link Structure) {
    is=title AnimGraphNode_OrientationWarping
    bUseManualRootMotionVelocity=true
    maxCorrectionDegrees=45
    mode=Graph
    rotationInterpSpeed=5
    alphaBoolBlend {blendInTime=0.2, blendOption=Linear, blendOutTime=0.2}
    iKFootBones [{boneName=ik_foot_l}, {boneName=ik_foot_r}]
    iKFootRootBone {boneName=ik_foot_root}
    spineBones [{boneName=spine_01}, {boneName=spine_02}, {boneName=spine_03}, {boneName=spine_04}, {boneName=spine_05}, {boneName=ik_hand_root}]
   }
   Output Pose(Result=Component To Local(ComponentPose=Reset Root Transform!(ComponentPose:Component Space Pose Link Structure=@n1, Alpha:float=1) -> (Pose:Component Space Pose Link Structure) {is=title AnimGraphNode_ResetRoot, alphaBoolBlend {blendOption=Linear}})) {layerGroup=None}
  ]
 }
 CrouchTransition {
  entry=MotionMatching
  path=AnimGraph.AnimGraphNode_StateMachine_0.CrouchTransition
  ids=id
  nodes [
   CrouchStart() {is=title AnimStateAliasNode, also {aliases [CrouchToStand, MotionMatching]}}
   CrouchEnd() {is=title AnimStateAliasNode, also {aliases [MotionMatching, StandToCrouch]}}
   ForceToMM() {is=title AnimStateAliasNode, also {aliases [CrouchToStand, StandToCrouch]}}
   AnimStateNode_0 = MotionMatching() {is=title AnimStateNode}
   AnimStateNode_1 = StandToCrouch() {is=title AnimStateNode, bAlwaysResetOnEntry=true}
   AnimStateNode_3 = CrouchToStand() {is=title AnimStateNode, bAlwaysResetOnEntry=true}
   AnimStateTransitionNode_0 = AnimStateTransitionNode() {bAutomaticRuleBasedOnSequencePlayerInState=true, priorityOrder=2}
   AnimStateTransitionNode_1 = AnimStateTransitionNode()
   AnimStateTransitionNode_10 = AnimStateTransitionNode() {crossfadeDuration=0.15, logicType=TLT_Inertialization}
   AnimStateTransitionNode_2 = AnimStateTransitionNode() {crossfadeDuration=0.15, logicType=TLT_Inertialization}
   AnimStateTransitionNode_3 = AnimStateTransitionNode() {bAutomaticRuleBasedOnSequencePlayerInState=true, priorityOrder=2}
  ]
  links [
   {
    dst=CrouchToStand
    rule=AnimGraph.AnimGraphNode_StateMachine_0.CrouchTransition.AnimStateTransitionNode_2.Transition
    src=CrouchEnd
    via=AnimStateTransitionNode_2
   }
   {
    dst=StandToCrouch
    rule=AnimGraph.AnimGraphNode_StateMachine_0.CrouchTransition.AnimStateTransitionNode_10.Transition
    src=CrouchStart
    via=AnimStateTransitionNode_10
   }
   {
    dst=MotionMatching
    rule=AnimGraph.AnimGraphNode_StateMachine_0.CrouchTransition.AnimStateTransitionNode_3.Transition
    src=CrouchToStand
    via=AnimStateTransitionNode_3
   }
   {
    dst=MotionMatching
    rule=AnimGraph.AnimGraphNode_StateMachine_0.CrouchTransition.AnimStateTransitionNode_1.Transition
    src=ForceToMM
    via=AnimStateTransitionNode_1
   }
   {
    dst=MotionMatching
    rule=AnimGraph.AnimGraphNode_StateMachine_0.CrouchTransition.AnimStateTransitionNode_0.Transition
    src=StandToCrouch
    via=AnimStateTransitionNode_0
   }
  ]
 }
 MotionMatching {
  path=AnimGraph.AnimGraphNode_StateMachine_0.CrouchTransition.AnimStateNode_0.MotionMatching
  nodes [Output Animation Pose(Result=Use cached pose 'LocomotionPose'())]
 }
 StandToCrouch {
  path=AnimGraph.AnimGraphNode_StateMachine_0.CrouchTransition.AnimStateNode_1.StandToCrouch
  nodes [
   n1 = "Sequence Player (additive) 'NoWep_Idle2Crouch'"() {
    bLoopAnimation=false
    groupName=CrouchStart
    method=SyncGroup
    sequence=/Game/Animation/Game/MainPlayerFN/Animation/Transitions/NoWep_Idle2Crouch
   }
   n2 = "Sequence Player (additive) 'NoWep_Idle2Crouch_InMotion'"() {
    bLoopAnimation=false
    groupName=CrouchStart
    method=SyncGroup
    sequence=/Game/Animation/Game/MainPlayerFN/Animation/Transitions/NoWep_Idle2Crouch_InMotion
   }
   n3 = Get bIsMoving2D()
   Output Animation Pose(Result=Apply Mesh Space Additive(
     Base=Use cached pose 'LocomotionPose'()
     Additive=Blend Poses by bool(BlendPose_0=@n2, BlendPose_1=@n1, BlendTime_0=0.3, BlendTime_1=0.3, bActiveValue=@n3) {blendType=Linear}
     Alpha=1
    ) {alphaBoolBlend {blendOption=Linear}})
  ]
 }
 CrouchToStand {
  path=AnimGraph.AnimGraphNode_StateMachine_0.CrouchTransition.AnimStateNode_3.CrouchToStand
  nodes [
   n1 = "Sequence Player (additive) 'NoWep_Crouch2Idle'"() {
    bLoopAnimation=false
    groupName=CrouchEnd
    method=SyncGroup
    sequence=/Game/Animation/Game/MainPlayerFN/Animation/Transitions/NoWep_Crouch2Idle
   }
   n2 = "Sequence Player (additive) 'NoWep_Crouch2Idle_InMotion'"() {
    bLoopAnimation=false
    groupName=CrouchEnd
    method=SyncGroup
    sequence=/Game/Animation/Game/MainPlayerFN/Animation/Transitions/NoWep_Crouch2Idle_InMotion
   }
   n3 = Get bIsMoving2D()
   Output Animation Pose(Result=Apply Mesh Space Additive(
     Base=Use cached pose 'LocomotionPose'()
     Additive=Blend Poses by bool(BlendPose_0=@n2, BlendPose_1=@n1, BlendTime_0=0.3, BlendTime_1=0.3, bActiveValue=@n3) {blendType=Linear}
     Alpha=1
    ) {alphaBoolBlend {blendOption=Linear}})
  ]
 }
 Transition {path=AnimGraph.AnimGraphNode_StateMachine_0.CrouchTransition.AnimStateTransitionNode_0.Transition, nodes [Result()]}
 Transition {
  path=AnimGraph.AnimGraphNode_StateMachine_0.CrouchTransition.AnimStateTransitionNode_1.Transition
  nodes [Result(bCanEnterTransition=Get bShouldExitCrouchTransition() -> (bShouldExitCrouchTransition:bool))]
 }
 Transition {
  path=AnimGraph.AnimGraphNode_StateMachine_0.CrouchTransition.AnimStateTransitionNode_10.Transition
  nodes [Result(bCanEnterTransition=Get bShouldPlayStandToCrouchTransition() -> (bShouldPlayStandToCrouchTransition:bool))]
 }
 Transition {
  path=AnimGraph.AnimGraphNode_StateMachine_0.CrouchTransition.AnimStateTransitionNode_2.Transition
  nodes [Result(bCanEnterTransition=Get bShouldPlayCrouchToStandTransition() -> (bShouldPlayCrouchToStandTransition:bool))]
 }
 Transition {path=AnimGraph.AnimGraphNode_StateMachine_0.CrouchTransition.AnimStateTransitionNode_3.Transition, nodes [Result()]}
 Locomotion {
  entry=Conduit
  path=AnimGraph.AnimGraphNode_StateMachine_1.Locomotion
  ids=id
  nodes [
   AnimStateConduitNode_0 = Conduit()
   AnimStateNode_0 = Idle()
   AnimStateNode_1 = Moving() {is=title AnimStateNode}
   AnimStateTransitionNode_0 = AnimStateTransitionNode() {crossfadeDuration=0.1, priorityOrder=3}
   AnimStateTransitionNode_1 = AnimStateTransitionNode() {crossfadeDuration=0.45}
   AnimStateTransitionNode_2 = AnimStateTransitionNode()
   AnimStateTransitionNode_3 = AnimStateTransitionNode()
   AnimStateTransitionNode_4 = AnimStateTransitionNode() {crossfadeDuration=0.1}
   AnimStateTransitionNode_5 = AnimStateTransitionNode() {bAllowInertializationForSelfTransitions=true, crossfadeDuration=0.3, priorityOrder=2}
  ]
  links [
   {
    dst=Idle
    rule=AnimGraph.AnimGraphNode_StateMachine_1.Locomotion.AnimStateTransitionNode_2.Transition
    src=Conduit
    via=AnimStateTransitionNode_2
   }
   {
    dst=Moving
    rule=AnimGraph.AnimGraphNode_StateMachine_1.Locomotion.AnimStateTransitionNode_3.Transition
    src=Conduit
    via=AnimStateTransitionNode_3
   }
   {
    dst=Moving
    rule=AnimGraph.AnimGraphNode_StateMachine_1.Locomotion.AnimStateTransitionNode_0.Transition
    src=Idle
    via=AnimStateTransitionNode_0
   }
   {
    dst=Moving
    rule=AnimGraph.AnimGraphNode_StateMachine_1.Locomotion.AnimStateTransitionNode_4.Transition
    src=Idle
    via=AnimStateTransitionNode_4
   }
   {
    dst=Moving
    rule=AnimGraph.AnimGraphNode_StateMachine_1.Locomotion.AnimStateTransitionNode_5.Transition
    src=Idle
    via=AnimStateTransitionNode_5
   }
   {
    dst=Idle
    rule=AnimGraph.AnimGraphNode_StateMachine_1.Locomotion.AnimStateTransitionNode_1.Transition
    src=Moving
    via=AnimStateTransitionNode_1
   }
  ]
 }
 Conduit {path=AnimGraph.AnimGraphNode_StateMachine_1.Locomotion.AnimStateConduitNode_0.Conduit, nodes [Result(bCanEnterTransition=true)]}
 Idle {
  path=AnimGraph.AnimGraphNode_StateMachine_1.Locomotion.AnimStateNode_0.Idle
  nodes [
   AnimGraphNode_StateMachine_0 = IdleState() {is=title AnimGraphNode_StateMachine}
   n1 = ItemAnimLayerInterface - ItemIdleOverride!(InPose_Idle:Pose Link Structure=@AnimGraphNode_StateMachine_0) -> (Pose:Pose Link Structure) {
    is=title AnimGraphNode_LinkedAnimLayer
    interface=/Game/Characters/Player/Male/Male_Avg_Base/AnimLayerInterface/ItemAnimLayerInterface.ItemAnimLayerInterface_C
    layer=ItemIdleOverride
   }
   Output Animation Pose(Result=Dead Blending!(Source:Pose Link Structure=@n1) -> (Pose:Pose Link Structure) {is=title AnimGraphNode_DeadBlending, defaultBlendMode=Linear})
  ]
 }
 IdleState {
  entry=IdleDefault
  path=AnimGraph.AnimGraphNode_StateMachine_1.Locomotion.AnimStateNode_0.Idle.AnimGraphNode_StateMachine_0.IdleState
  ids=id
  nodes [
   AnimStateNode_1 = ShuffleInPlace() {is=title AnimStateNode}
   AnimStateNode_3 = IdleDefault() {
    is=title AnimStateNode
    stateEntered {guid=#1}
    stateFullyBlended {guid=#2}
    stateLeft {guid=#3}
   }
   AnimStateNode_4 = IdleBreak() {is=title AnimStateNode}
   AnimStateTransitionNode_11 = AnimStateTransitionNode() {crossfadeDuration=0.35}
   AnimStateTransitionNode_12 = AnimStateTransitionNode()
   AnimStateTransitionNode_3 = AnimStateTransitionNode() {crossfadeDuration=0.35, logicType=TLT_Inertialization}
   AnimStateTransitionNode_6 = AnimStateTransitionNode() {crossfadeDuration=0.3}
   AnimStateTransitionNode_7 = AnimStateTransitionNode() {crossfadeDuration=0.401}
   AnimStateTransitionNode_8 = AnimStateTransitionNode() {
    logicType=TLT_Inertialization
    priorityOrder=2
    blendProfileWrapper {blendProfile=&Fortnite_M_Avg_Player_Skeleton:SlowFeetFastUpper}
   }
   AnimStateTransitionNode_9 = AnimStateTransitionNode() {
    bAutomaticRuleBasedOnSequencePlayerInState=true
    transitionEnd {guid=#4}
    transitionInterrupt {guid=#5}
    transitionStart {guid=#6}
   }
  ]
  links [
   {
    dst=IdleDefault
    rule=AnimGraph.AnimGraphNode_StateMachine_1.Locomotion.AnimStateNode_0.Idle.AnimGraphNode_StateMachine_0.IdleState.AnimStateTransitionNode_7.Transition
    src=IdleBreak
    via=AnimStateTransitionNode_7
   }
   {
    dst=IdleDefault
    rule=AnimGraph.AnimGraphNode_StateMachine_1.Locomotion.AnimStateNode_0.Idle.AnimGraphNode_StateMachine_0.IdleState.AnimStateTransitionNode_8.Transition
    src=IdleBreak
    via=AnimStateTransitionNode_8
   }
   {
    dst=IdleDefault
    rule=AnimGraph.AnimGraphNode_StateMachine_1.Locomotion.AnimStateNode_0.Idle.AnimGraphNode_StateMachine_0.IdleState.AnimStateTransitionNode_9.Transition
    src=IdleBreak
    via=AnimStateTransitionNode_9
   }
   {
    dst=IdleBreak
    rule=AnimGraph.AnimGraphNode_StateMachine_1.Locomotion.AnimStateNode_0.Idle.AnimGraphNode_StateMachine_0.IdleState.AnimStateTransitionNode_6.Transition
    src=IdleDefault
    via=AnimStateTransitionNode_6
   }
   {
    dst=ShuffleInPlace
    rule=AnimGraph.AnimGraphNode_StateMachine_1.Locomotion.AnimStateNode_0.Idle.AnimGraphNode_StateMachine_0.IdleState.AnimStateTransitionNode_11.Transition
    src=IdleDefault
    via=AnimStateTransitionNode_11
   }
   {
    dst=ShuffleInPlace
    rule=AnimGraph.AnimGraphNode_StateMachine_1.Locomotion.AnimStateNode_0.Idle.AnimGraphNode_StateMachine_0.IdleState.AnimStateTransitionNode_12.Transition
    src=IdleDefault
    via=AnimStateTransitionNode_12
   }
   {
    dst=IdleDefault
    rule=AnimGraph.AnimGraphNode_StateMachine_1.Locomotion.AnimStateNode_0.Idle.AnimGraphNode_StateMachine_0.IdleState.AnimStateTransitionNode_3.AnimationTransitionGraph_1
    src=ShuffleInPlace
    via=AnimStateTransitionNode_3
   }
  ]
 }
 ShuffleInPlace {
  path=AnimGraph.AnimGraphNode_StateMachine_1.Locomotion.AnimStateNode_0.Idle.AnimGraphNode_StateMachine_0.IdleState.AnimStateNode_1.ShuffleInPlace
  nodes [
   n1 = Break Fort Anim Input Turn in Place!(FortAnimInput_TurnInPlace:"Fort Anim Input Turn In Place Structure (by ref)"=Get TurnInPlaceInput() -> (TurnInPlaceInput:Fort Anim Input Turn In Place Structure)) -> (
    RootYawOffset:float=~
    TurnAnimPlayRate:float=~
    bWantsToTurn:bool=~
    bExitTurn:bool=~
    bUseTurnA:bool=~
    TurnACurveAlpha:float=~
    TurnBCurveAlpha:float=~
    bWantsToTurnInVehicle:bool=~
    bWantsToTurnAgain:bool=~
    bTurningLeft:bool=~
    ShufflePlayRate:float=~
    CrouchShufflePlayRate:float=~
    CrouchShuffleStartPosition:float=~
    CrouchShuffleSpeedScaling:float=~
    bTransition_LowerBody_Shuffle:bool=~
    bTransition_LowerBody_ShuffleADS:bool=~
    bPlayOutro:bool=~
   ) {is=title K2Node_BreakStruct}
   n2 = Sequence Player 'NoWep_Crouch_Loop_E_LFF'!(PlayRate:float=@n1.CrouchShufflePlayRate, StartPosition:float=@n1.CrouchShuffleStartPosition) -> (Pose) {
    groupName=TurnInPlaceSync
    method=SyncGroup
    sequence=/LocomotionInitiative/Animation/Locomotion_anims/MotionMatching/Crouch/FBX/NoWep_Crouch_Loop_E_LFF
    playRateScaleBiasClampConstants {bClampResult=true, clampMax=1.2, clampMin=-1.2, scale=0.15}
   }
   n3 = Speed Warping!(
    ComponentPose:Component Space Pose Link Structure=Local To Component(LocalPose=@n2)
    Alpha:float=1
    SpeedScaling:float=@n1.CrouchShuffleSpeedScaling
    ManualSpeedWarpingDir:Vector="0,1, 0"
   ) -> (Pose:Component Space Pose Link Structure) {
    is=title FortAnimGraphNode_SpeedWarping
    lODThreshold=3
    alphaBoolBlend {blendOption=Linear}
    feetDefinitions [
     {numBonesInLimb=2, fKFootBone {boneName=foot_l}, iKFootBone {boneName=ik_foot_l}}
     {numBonesInLimb=2, fKFootBone {boneName=foot_r}, iKFootBone {boneName=ik_foot_r}}
    ]
    iKFootRootBone {boneName=ik_foot_root}
    pelvisBone {boneName=pelvis}
    speedScalingScaleBiasClamp {bClampResult=true, clampMin=0.5, scale=0.08}
   }
   n4 = "Transform (Modify) Bone - Bone: pelvis"!(ComponentPose=@n3, bAlphaBoolEnabled:bool=Get bPlayUpperBodyTargeting() -> (bPlayUpperBodyTargeting:bool), Translation:Vector="0,0,-10") -> (Pose) {
    alphaInputType=Bool
    lODThreshold=2
    translationMode=BMM_Additive
    alphaBoolBlend {blendInTime=0.1, blendOption=Linear, blendOutTime=0.1}
    boneToModify {boneName=pelvis}
    translation {z=-10}
   }
   n5 = Modify Curve(SourcePose=Component To Local(ComponentPose=@n4), CurveValues_0=1, Alpha=1) {curveNames [CrouchShuffle], curveValues [0]}
   n6 = Sequence Player 'M_Jog_Melee_Twohanded_E_Core'!(PlayRate:float=@n1.ShufflePlayRate) -> (Pose) {
    groupName=TurnInPlaceSync
    method=SyncGroup
    sequence=/Game/Animation/Game/MainPlayer/Locomotion/Medium/Male/Jog/Melee_TH/M_Jog_Melee_Twohanded_E_Core
    playRateScaleBiasClampConstants {bClampResult=true, clampMax=1.2, clampMin=-1.2, scale=0.9}
   }
   n7 = Sequence Player 'M_Jog_Melee_TwoHanded_E_ScaledForTurnInPlace'!(PlayRate:float=@n1.ShufflePlayRate) -> (Pose) {
    groupName=TurnInPlaceSync
    method=SyncGroup
    sequence=/Game/Animation/Game/MainPlayer/Locomotion/Medium/Male/Jog/Melee_TH/M_Jog_Melee_TwoHanded_E_ScaledForTurnInPlace
    playRateScaleBiasClampConstants {bClampResult=true, clampMax=1.2, clampMin=-1.2, scale=0.18}
   }
   n8 = Two Way Blend(A=@n7, B=@n6, Alpha=Get AbsYawDeltaSmoothed() -> (AbsYawDeltaSmoothed:float)) {blendNode {alphaBoolBlend {blendOption=Linear}, alphaScaleBias {scale=0.0625}}}
   n9 = Get bIsCrouching()
   Output Animation Pose(Result=Rotate Root Bone!(
     BasePose:Pose Link Structure=Blend Poses by bool(BlendPose_0=@n5, BlendPose_1=@n8, bActiveValue=@n9) {blendType=Linear}
     Pitch:float=0
     Yaw:float=Get MeleeTwistCurveValue() -> (MeleeTwistCurveValue:float)
     MeshToComponent:Rotator="0,-90, 0"
    ) -> (Pose:Pose Link Structure) {is=title AnimGraphNode_RotateRootBone, yawScaleBiasClamp {bClampResult=true, clampMax=60, clampMin=-60, scale=0.75}})
  ]
 }
 IdleDefault {
  path=AnimGraph.AnimGraphNode_StateMachine_1.Locomotion.AnimStateNode_0.Idle.AnimGraphNode_StateMachine_0.IdleState.AnimStateNode_3.IdleDefault
  nodes [
   n1 = Sequence Player!(Sequence:Anim Sequence Base Object Reference) -> (Pose)
   n2 = Break Motion Matching LODConfig!(MotionMatchingLODConfig:"Motion Matching LODConfig Structure (by ref)"=Get CurrentLODConfig() -> (CurrentLODConfig:Motion Matching LODConfig Structure)) -> (
    MotionDatabaseProxyTable:Proxy Table Object Reference
    StandIdleLoop:Anim Sequence Object Reference
    CrouchIdleLoop:Anim Sequence Object Reference
   ) {is=title K2Node_BreakStruct}
   n3 = Sequence Player(Sequence=@n2.StandIdleLoop)
   n4 = Get bHasStandIdleOverride() -> (bHasStandIdleOverride:bool)
   Output Animation Pose(Result=Blend Poses by bool(
     BlendPose_0=Sequence Player(Sequence=@n2.CrouchIdleLoop)
     BlendPose_1=Blend Poses by bool(BlendPose_0=@n1, BlendPose_1=@n3, bActiveValue=@n4)
     BlendTime_0=0
     BlendTime_1=0
     bActiveValue=Get bIsCrouching()
    ) {blendType=Linear, childUpateMode=ResetChildOnActivate})
  ]
 }
 IdleBreak {
  path=AnimGraph.AnimGraphNode_StateMachine_1.Locomotion.AnimStateNode_0.Idle.AnimGraphNode_StateMachine_0.IdleState.AnimStateNode_4.IdleBreak
  nodes [
   n1 = Sequence Player(Sequence=Get SelectedIdleBreak() -> (SelectedIdleBreak:Anim Sequence Object Reference)) {node {updateFunction {functionName=OnUpdateIdleBreaks}}, updateFunction {bSelfContext=true, memberName=OnUpdateIdleBreaks}}
   Output Animation Pose(Result=@n1)
  ]
 }
 Transition {
  path=AnimGraph.AnimGraphNode_StateMachine_1.Locomotion.AnimStateNode_0.Idle.AnimGraphNode_StateMachine_0.IdleState.AnimStateTransitionNode_11.Transition
  nodes [Result(bCanEnterTransition=Break Fort Anim Input Turn in Place(FortAnimInput_TurnInPlace=Get TurnInPlaceInput()) -> (
     RootYawOffset=~
     TurnAnimPlayRate=~
     bWantsToTurn=~
     bExitTurn=~
     bUseTurnA=~
     TurnACurveAlpha=~
     TurnBCurveAlpha=~
     bWantsToTurnInVehicle=~
     bWantsToTurnAgain=~
     bTurningLeft=~
     ShufflePlayRate=~
     CrouchShufflePlayRate=~
     CrouchShuffleStartPosition=~
     CrouchShuffleSpeedScaling=~
     bTransition_LowerBody_Shuffle=~
     bTransition_LowerBody_ShuffleADS=~
     bPlayOutro=~
    ).bTransition_LowerBody_ShuffleADS)]
 }
 Transition {
  path=AnimGraph.AnimGraphNode_StateMachine_1.Locomotion.AnimStateNode_0.Idle.AnimGraphNode_StateMachine_0.IdleState.AnimStateTransitionNode_12.Transition
  nodes [Result(bCanEnterTransition=Break Fort Anim Input Turn in Place(FortAnimInput_TurnInPlace=Get TurnInPlaceInput()) -> (
     RootYawOffset=~
     TurnAnimPlayRate=~
     bWantsToTurn=~
     bExitTurn=~
     bUseTurnA=~
     TurnACurveAlpha=~
     TurnBCurveAlpha=~
     bWantsToTurnInVehicle=~
     bWantsToTurnAgain=~
     bTurningLeft=~
     ShufflePlayRate=~
     CrouchShufflePlayRate=~
     CrouchShuffleStartPosition=~
     CrouchShuffleSpeedScaling=~
     bTransition_LowerBody_Shuffle=~
     bTransition_LowerBody_ShuffleADS=~
     bPlayOutro=~
    ).bTransition_LowerBody_Shuffle)]
 }
 AnimationTransitionGraph_1 {
  path=AnimGraph.AnimGraphNode_StateMachine_1.Locomotion.AnimStateNode_0.Idle.AnimGraphNode_StateMachine_0.IdleState.AnimStateTransitionNode_3.AnimationTransitionGraph_1
  nodes [
   n1 = Get TurnInPlaceInput()
   Result(bCanEnterTransition=NOT Boolean(A=Break Fort Anim Input Turn in Place(FortAnimInput_TurnInPlace=@n1) -> (
      RootYawOffset=~
      TurnAnimPlayRate=~
      bWantsToTurn=~
      bExitTurn=~
      bUseTurnA=~
      TurnACurveAlpha=~
      TurnBCurveAlpha=~
      bWantsToTurnInVehicle=~
      bWantsToTurnAgain=~
      bTurningLeft=~
      ShufflePlayRate=~
      CrouchShufflePlayRate=~
      CrouchShuffleStartPosition=~
      CrouchShuffleSpeedScaling=~
      bTransition_LowerBody_Shuffle=~
      bTransition_LowerBody_ShuffleADS=~
      bPlayOutro=~
     ).bTransition_LowerBody_Shuffle))
  ]
 }
 Transition {
  path=AnimGraph.AnimGraphNode_StateMachine_1.Locomotion.AnimStateNode_0.Idle.AnimGraphNode_StateMachine_0.IdleState.AnimStateTransitionNode_6.Transition
  nodes [Result(bCanEnterTransition=Get bShouldPlayIdleBreak() -> (bShouldPlayIdleBreak:bool))]
 }
 Transition {
  path=AnimGraph.AnimGraphNode_StateMachine_1.Locomotion.AnimStateNode_0.Idle.AnimGraphNode_StateMachine_0.IdleState.AnimStateTransitionNode_7.Transition
  nodes [Result(bCanEnterTransition=Get bShouldExitIdleBreakSlow() -> (bShouldExitIdleBreakSlow:bool))]
 }
 Transition {
  path=AnimGraph.AnimGraphNode_StateMachine_1.Locomotion.AnimStateNode_0.Idle.AnimGraphNode_StateMachine_0.IdleState.AnimStateTransitionNode_8.Transition
  nodes [Result(bCanEnterTransition=Get bShouldExitIdleBreak() -> (bShouldExitIdleBreak:bool))]
 }
 Transition {
  path=AnimGraph.AnimGraphNode_StateMachine_1.Locomotion.AnimStateNode_0.Idle.AnimGraphNode_StateMachine_0.IdleState.AnimStateTransitionNode_9.Transition
  nodes [Result()]
 }
 Moving {
  path=AnimGraph.AnimGraphNode_StateMachine_1.Locomotion.AnimStateNode_1.Moving
  nodes [
   n1 = Get MeleeTwistBlendSpacePawnSpeed() -> (MeleeTwistBlendSpacePawnSpeed:float)
   n2 = Get LocalVelocityYawAngle() -> (LocalVelocityYawAngle:float)
   n3 = Apply Additive(
    Base=Rotate Root Bone(
     BasePose=Blendspace Player 'Crouch_NoWep_Core'(X=@n2, Y=@n1) {blendSpace=/Game/Animation/Game/MainPlayer/Locomotion/Medium/Male/Crouch/NoWeapon/Crouch_NoWep_Core}
     Pitch=0
     Yaw=Get MeleeTwistCurveValue()
     MeshToComponent="0,-90,0"
    ) {yawScaleBiasClamp {scale=0.35}}
    Additive="Sequence Player (additive) 'Crouch_MeleeCombo_PelvisCorrection'"() {bLoopAnimation=false, sequence=/Game/Animation/Game/MainPlayerFN/Poses/AdditiveCorrection/Crouch_MeleeCombo_PelvisCorrection}
   ) {alphaBoolBlend {blendInTime=0.2, blendOption=Linear, blendOutTime=0.3}}
   n4 = Rotate Root Bone(
    BasePose=Blendspace Player 'M_Jog_NoWep_Core_No45s'(X=Get LocalVelocityYawAngleMinusMeleeTwist() -> (LocalVelocityYawAngleMinusMeleeTwist:float), Y=Get MeleeTwistBlendSpacePawnSpeed()) {blendSpace=/Game/Animation/Game/MainPlayer/Locomotion/Medium/Male/Jog/NoWeapon/M_Jog_NoWep_Core_No45s}
    Pitch=0
    Yaw=Get MeleeTwistCurveValue()
    MeshToComponent="0,-90,0"
   )
   n5 = Get bIsCrouching()
   n6 = Get SprintYawAngle() -> (SprintYawAngle:float)
   n7 = Layered blend per bone(
    BasePose=Use cached pose 'LocomotionBase'()
    BlendPoses_0=Apply Additive(
     Base=Use cached pose 'LocomotionBase'()
     Additive=Blendspace Player 'Run_Lean_BS'(X=@n6, Y=0) {blendSpace=/Game/Animation/Game/MainPlayerFN/Poses/LocomotionLeans/Run_Lean_BS}
    ) {alphaBoolBlend {blendOption=Linear}}
    BlendWeights_0=Get RunLeanAlpha() -> (RunLeanAlpha:float)
   ) {layerSetup [{branchFilters [
       {boneName=Root}
       {blendDepth=-1, boneName=pelvis}
       {boneName=thigh_l}
       {boneName=thigh_r}
       {blendDepth=-1, boneName=ik_foot_root}
       {boneName=Spine_01}
       {blendDepth=-1, boneName=neck_01}
      ]}]}
   n8 = Use cached pose 'LocomotionBase'()
   n9 = Get bEnableAdditiveRunLeans() -> (bEnableAdditiveRunLeans:bool)
   Output Animation Pose(Result=Blend Poses by bool(
     BlendPose_0=Blend Poses by bool(BlendPose_0=@n3, BlendPose_1=@n4, bActiveValue=@n5) {blendType=Linear}
     BlendPose_1=Blend Poses by bool(BlendPose_0=@n7, BlendPose_1=@n8, bActiveValue=@n9) {blendType=Linear}
     BlendTime_0=0.2
     BlendTime_1=0.2
     bActiveValue=Get bUseLocomotionBlendspace() -> (bUseLocomotionBlendspace:bool)
    ) {blendType=Linear})
  ]
 }
 Transition {
  path=AnimGraph.AnimGraphNode_StateMachine_1.Locomotion.AnimStateTransitionNode_0.Transition
  nodes [Result(bCanEnterTransition=Get bTransition_LowerBody_Movement() -> (bTransition_LowerBody_Movement:bool))]
 }
 Transition {
  path=AnimGraph.AnimGraphNode_StateMachine_1.Locomotion.AnimStateTransitionNode_1.Transition
  nodes [Result(bCanEnterTransition=Get bShouldPlayIdle() -> (bShouldPlayIdle:bool))]
 }
 Transition {
  path=AnimGraph.AnimGraphNode_StateMachine_1.Locomotion.AnimStateTransitionNode_2.Transition
  nodes [Result(bCanEnterTransition=Get bShouldPlayIdle())]
 }
 Transition {
  path=AnimGraph.AnimGraphNode_StateMachine_1.Locomotion.AnimStateTransitionNode_3.Transition
  nodes [Result(bCanEnterTransition=NOT Boolean(A=Get bShouldPlayIdle()))]
 }
 Transition {
  path=AnimGraph.AnimGraphNode_StateMachine_1.Locomotion.AnimStateTransitionNode_4.Transition
  nodes [Result(bCanEnterTransition=Get bIdleToMotionMatchingNoInertialization() -> (bIdleToMotionMatchingNoInertialization:bool))]
 }
 Transition {
  path=AnimGraph.AnimGraphNode_StateMachine_1.Locomotion.AnimStateTransitionNode_5.Transition
  nodes [Result(bCanEnterTransition=Get bIdleToMotionMatchingCrouchTransition() -> (bIdleToMotionMatchingCrouchTransition:bool))]
 }
 LowLODMovement {
  entry=Conduit
  path=AnimGraph.AnimGraphNode_StateMachine_2.LowLODMovement
  ids=id
  nodes [
   AnimStateConduitNode_0 = Conduit()
   AnimStateNode_0 = Crouch() {is=title AnimStateNode}
   AnimStateNode_1 = Walk() {is=title AnimStateNode}
   AnimStateNode_2 = Jog() {is=title AnimStateNode}
   AnimStateNode_3 = Run() {is=title AnimStateNode}
   AnimStateNode_4 = Sprint() {is=title AnimStateNode}
   AnimStateNode_5 = Idle()
   AnimStateTransitionNode_0 = AnimStateTransitionNode()
   AnimStateTransitionNode_1 = AnimStateTransitionNode() {priorityOrder=2}
   AnimStateTransitionNode_10 = AnimStateTransitionNode()
   AnimStateTransitionNode_11 = AnimStateTransitionNode()
   AnimStateTransitionNode_12 = AnimStateTransitionNode()
   AnimStateTransitionNode_2 = AnimStateTransitionNode() {priorityOrder=2}
   AnimStateTransitionNode_3 = AnimStateTransitionNode() {priorityOrder=2}
   AnimStateTransitionNode_4 = AnimStateTransitionNode() {priorityOrder=2}
   AnimStateTransitionNode_5 = AnimStateTransitionNode() {priorityOrder=100}
   AnimStateTransitionNode_7 = AnimStateTransitionNode()
   AnimStateTransitionNode_8 = AnimStateTransitionNode()
   AnimStateTransitionNode_9 = AnimStateTransitionNode()
  ]
  links [
   {
    dst=Crouch
    rule=AnimGraph.AnimGraphNode_StateMachine_2.LowLODMovement.AnimStateTransitionNode_0.Transition
    src=Conduit
    via=AnimStateTransitionNode_0
   }
   {
    dst=Idle
    rule=AnimGraph.AnimGraphNode_StateMachine_2.LowLODMovement.AnimStateTransitionNode_5.Transition
    src=Conduit
    via=AnimStateTransitionNode_5
   }
   {
    dst=Jog
    rule=AnimGraph.AnimGraphNode_StateMachine_2.LowLODMovement.AnimStateTransitionNode_2.Transition
    src=Conduit
    via=AnimStateTransitionNode_2
   }
   {
    dst=Run
    rule=AnimGraph.AnimGraphNode_StateMachine_2.LowLODMovement.AnimStateTransitionNode_3.Transition
    src=Conduit
    via=AnimStateTransitionNode_3
   }
   {
    dst=Sprint
    rule=AnimGraph.AnimGraphNode_StateMachine_2.LowLODMovement.AnimStateTransitionNode_4.Transition
    src=Conduit
    via=AnimStateTransitionNode_4
   }
   {
    dst=Walk
    rule=AnimGraph.AnimGraphNode_StateMachine_2.LowLODMovement.AnimStateTransitionNode_1.Transition
    src=Conduit
    via=AnimStateTransitionNode_1
   }
   {
    dst=Conduit
    rule=AnimGraph.AnimGraphNode_StateMachine_2.LowLODMovement.AnimStateTransitionNode_12.Transition
    src=Crouch
    via=AnimStateTransitionNode_12
   }
   {
    dst=Conduit
    rule=AnimGraph.AnimGraphNode_StateMachine_2.LowLODMovement.AnimStateTransitionNode_7.Transition
    src=Idle
    via=AnimStateTransitionNode_7
   }
   {
    dst=Conduit
    rule=AnimGraph.AnimGraphNode_StateMachine_2.LowLODMovement.AnimStateTransitionNode_10.Transition
    src=Jog
    via=AnimStateTransitionNode_10
   }
   {
    dst=Conduit
    rule=AnimGraph.AnimGraphNode_StateMachine_2.LowLODMovement.AnimStateTransitionNode_9.Transition
    src=Run
    via=AnimStateTransitionNode_9
   }
   {
    dst=Conduit
    rule=AnimGraph.AnimGraphNode_StateMachine_2.LowLODMovement.AnimStateTransitionNode_8.Transition
    src=Sprint
    via=AnimStateTransitionNode_8
   }
   {
    dst=Conduit
    rule=AnimGraph.AnimGraphNode_StateMachine_2.LowLODMovement.AnimStateTransitionNode_11.Transition
    src=Walk
    via=AnimStateTransitionNode_11
   }
  ]
 }
 Conduit {path=AnimGraph.AnimGraphNode_StateMachine_2.LowLODMovement.AnimStateConduitNode_0.Conduit, nodes [Result(bCanEnterTransition=true)]}
 Crouch {
  path=AnimGraph.AnimGraphNode_StateMachine_2.LowLODMovement.AnimStateNode_0.Crouch
  nodes [
   n1 = Get LocalVelocityYawAngle()
   n2 = Get PlayRateMultiplier()
   Blendspace Player 'LowLOD_Crouch_Locomotion_BS'!(X=@n1, Y=Get Speed2D() -> (Speed2D:float), PlayRate:float=@n2) -> (Pose) {blendSpace=/LocomotionInitiative/Animation/BlendSpaces/LowLOD_Crouch_Locomotion_BS}
   Output Animation Pose(Result=Blendspace Player 'LowLOD_Crouch_Locomotion_BS_1D'!(X=@n1, PlayRate:float=@n2) -> (Pose) {blendSpace=/LocomotionInitiative/Animation/BlendSpaces/LowLOD_Crouch_Locomotion_BS_1D})
  ]
 }
 Walk {
  path=AnimGraph.AnimGraphNode_StateMachine_2.LowLODMovement.AnimStateNode_1.Walk
  nodes [Output Animation Pose(Result=Blendspace Player 'LowLOD_Walk_Locomotion_BS'!(X=Get LocalVelocityYawAngle(), Y=Get Speed2D(), PlayRate:float=Get PlayRateMultiplier()) -> (Pose) {blendSpace=/LocomotionInitiative/Animation/BlendSpaces/LowLOD_Walk_Locomotion_BS})]
 }
 Jog {
  path=AnimGraph.AnimGraphNode_StateMachine_2.LowLODMovement.AnimStateNode_2.Jog
  nodes [Output Animation Pose(Result=Blendspace Player 'LowLOD_Jog_Locomotion_BS'!(X=Get LocalVelocityYawAngle(), Y=Get Speed2D(), PlayRate:float=Get PlayRateMultiplier()) -> (Pose) {blendSpace=/LocomotionInitiative/Animation/BlendSpaces/LowLOD_Jog_Locomotion_BS})]
 }
 Run {
  path=AnimGraph.AnimGraphNode_StateMachine_2.LowLODMovement.AnimStateNode_3.Run
  nodes [Output Animation Pose(Result=Blendspace Player 'LowLOD_Run_Locomotion_BS'!(X=Get LocalVelocityYawAngle(), Y=Get Speed2D(), PlayRate:float=Get PlayRateMultiplier()) -> (Pose) {blendSpace=/LocomotionInitiative/Animation/BlendSpaces/LowLOD_Run_Locomotion_BS})]
 }
 Sprint {
  path=AnimGraph.AnimGraphNode_StateMachine_2.LowLODMovement.AnimStateNode_4.Sprint
  nodes [Output Animation Pose(Result=Blendspace Player 'LowLOD_Sprint_Locomotion_BS'!(X=Get LocalVelocityYawAngle(), Y=Get Speed2D(), PlayRate:float=Get PlayRateMultiplier()) -> (Pose) {blendSpace=/LocomotionInitiative/Animation/BlendSpaces/LowLOD_Sprint_Locomotion_BS})]
 }
 Idle {
  path=AnimGraph.AnimGraphNode_StateMachine_2.LowLODMovement.AnimStateNode_5.Idle
  nodes [
   n1 = Break Motion Matching LODConfig(MotionMatchingLODConfig=Get CurrentLODConfig())
   Output Animation Pose(Result=Blend Poses by bool(
     BlendPose_0=Sequence Player(Sequence=@n1.CrouchIdleLoop)
     BlendPose_1=Sequence Player(Sequence=@n1.StandIdleLoop)
     BlendTime_0=0
     BlendTime_1=0
     bActiveValue=Get bIsCrouching()
    ) {blendType=Linear, childUpateMode=ResetChildOnActivate})
  ]
 }
 Transition {
  path=AnimGraph.AnimGraphNode_StateMachine_2.LowLODMovement.AnimStateTransitionNode_0.Transition
  nodes [Result(bCanEnterTransition=Get bIsCrouchMoving() -> (bIsCrouchMoving:bool))]
 }
 Transition {
  path=AnimGraph.AnimGraphNode_StateMachine_2.LowLODMovement.AnimStateTransitionNode_1.Transition
  nodes [Result(bCanEnterTransition=Get bIsWalking() -> (bIsWalking:bool))]
 }
 Transition {path=AnimGraph.AnimGraphNode_StateMachine_2.LowLODMovement.AnimStateTransitionNode_10.Transition, nodes [Result(bCanEnterTransition=true)]}
 Transition {path=AnimGraph.AnimGraphNode_StateMachine_2.LowLODMovement.AnimStateTransitionNode_11.Transition, nodes [Result(bCanEnterTransition=true)]}
 Transition {path=AnimGraph.AnimGraphNode_StateMachine_2.LowLODMovement.AnimStateTransitionNode_12.Transition, nodes [Result(bCanEnterTransition=true)]}
 Transition {
  path=AnimGraph.AnimGraphNode_StateMachine_2.LowLODMovement.AnimStateTransitionNode_2.Transition
  nodes [Result(bCanEnterTransition=Get bIsJogging() -> (bIsJogging:bool))]
 }
 Transition {
  path=AnimGraph.AnimGraphNode_StateMachine_2.LowLODMovement.AnimStateTransitionNode_3.Transition
  nodes [Result(bCanEnterTransition=Get bIsRunning() -> (bIsRunning:bool))]
 }
 Transition {
  path=AnimGraph.AnimGraphNode_StateMachine_2.LowLODMovement.AnimStateTransitionNode_4.Transition
  nodes [Result(bCanEnterTransition=Get bIsTacticalSprint() -> (bIsTacticalSprint:bool))]
 }
 Transition {path=AnimGraph.AnimGraphNode_StateMachine_2.LowLODMovement.AnimStateTransitionNode_5.Transition, nodes [Result(bCanEnterTransition=true)]}
 Transition {path=AnimGraph.AnimGraphNode_StateMachine_2.LowLODMovement.AnimStateTransitionNode_7.Transition, nodes [Result(bCanEnterTransition=true)]}
 Transition {path=AnimGraph.AnimGraphNode_StateMachine_2.LowLODMovement.AnimStateTransitionNode_8.Transition, nodes [Result(bCanEnterTransition=true)]}
 Transition {path=AnimGraph.AnimGraphNode_StateMachine_2.LowLODMovement.AnimStateTransitionNode_9.Transition, nodes [Result(bCanEnterTransition=true)]}
 EventGraph {nodes [
   n1 = Branch(Condition=Get bShouldExitIdleBreak())
   > Set IdleBreakTimer(IdleBreakTimer:float=0) -> (Output_Get:float)
   Branch(execute=@n1.else, Condition="float >= float"(A=Get IdleBreakTimer() -> (IdleBreakTimer:float), B=15))
   > Set IdleBreakTimer(IdleBreakTimer=0)
   > Set IdleBreakTimer(IdleBreakTimer=float + float(A=Get IdleBreakTimer(), B=GetWorldDeltaSeconds!() -> (ReturnValue:double)))
   n2 = Branch(Condition="float >= float"(A=Get IdleBreakTimer(), B=5))
   > Do Once(Reset=@n2.else)
   > Set bShouldPlayIdleBreak(bShouldPlayIdleBreak:bool=true) -> (Output_Get:bool)
   > Delay(Duration=0.2)
   > Set bShouldPlayIdleBreak(bShouldPlayIdleBreak=false)
   n3 = Set IsCrouchTransition(IsCrouchTransition:bool="float > float"(A=GetCurveValue(CurveName=CrouchTransitionInPlace))) -> (Output_Get:bool)
   > Set bShouldExitIdleBreak(bShouldExitIdleBreak:bool=OR Boolean!(
     A=Get bIsMoving2D()
     B=Get bRecentlyFired() -> (bRecentlyFired:bool)
     C:bool=Get bRecentlyFiredAbility() -> (bRecentlyFiredAbility:bool)
     D:bool=Get bLocomotion_Idles_to_Turns() -> (bLocomotion_Idles_to_Turns:bool)
     E:bool=Get bIsMontagePlaying() -> (bIsMontagePlaying:bool)
     F:bool=Get bPlayUpperBodyTargeting()
     G:bool=@n3
    ) -> (ReturnValue)) -> (Output_Get:bool)
   Set TorsoLocomotionWeight(TorsoLocomotionWeight:float=GetCurveValue(CurveName=ApplyTorsoLocomotion)) -> (Output_Get:float)
   > Set ArmLeftLocomotionWeight(ArmLeftLocomotionWeight:float=GetCurveValue(CurveName=ApplyArmLeftLocomotion)) -> (Output_Get:float)
   > Set ArmRightLocomotionWeight(ArmRightLocomotionWeight:float=GetCurveValue(CurveName=ApplyArmRightLocomotion)) -> (Output_Get:float)
   > Set LeftFingersWepAlpha(LeftFingersWepAlpha:float=GetCurveValue(CurveName=ApplySourcePoseFingersWeaponL)) -> (Output_Get:float)
   > Set RightFingersWepAlpha(RightFingersWepAlpha:float=GetCurveValue(CurveName=ApplySourcePoseFingersWeaponR)) -> (Output_Get:float)
   n4 = NOT Boolean(A=Get bIsCrouching())
   n5 = Get bBothHandsOnWeapon() -> (bBothHandsOnWeapon:bool)
   RandomIntegerInRange!(Min:int=SelectInt(A=1, B=0, bPickA=AND Boolean(A=@n5, B=@n4)), Max:int=5) -> (ReturnValue:int=0)
   Set bShouldPlayIdle(bShouldPlayIdle:bool=WasAnimNotifyStateActiveInAnyState!(AnimNotifyStateType:Anim Notify State Class Reference=/LocomotionInitiative/Animation/Anim_NotifyStates/AnimNotifyState_BranchOutWindow.AnimNotifyState_BranchOutWindow_C) -> (ReturnValue:bool)) -> (Output_Get:bool)
  ]}
 ItemAimOffset {nodes [
   Input Pose!() -> (Pose, AimOffsetAlpha:double, Yaw:double, Pitch:double) {name=InPose_AimOffset}
   Input Pose!() -> (Pose) {name=InPose_PreAimOffsets}
   Output Pose() {is=title AnimGraphNode_Root, layerGroup=Item Override}
  ]}
 ItemFallAdditive {nodes [
   Input Pose() {name=InPoseItemFallAdditive}
   Output Pose() {layerGroup=Item Override}
  ]}
 ItemFallLandAdditive {nodes [
   Input Pose() {name=InPoseItemFallLandAdditive}
   Output Pose() {layerGroup=Item Override}
  ]}
 ItemFallLevel2Override {nodes [
   Input Pose() {name=InPose_FallLevel2}
   Output Pose() {layerGroup=Item Override}
  ]}
 ItemFinalPoseOverride {nodes [
   Input Pose() {name=InFinalPose}
   Output Pose() {layerGroup=Item Override}
  ]}
 ItemFullBodyAimOffsets {nodes [
   Input Pose!() -> (Pose, FullBody_AimOffsetAlpha:double, Aim_Yaw:double, Aim_Pitch:double) {name=InPoseFullBodyAimOffsets}
   Input Pose!() -> (Pose) {name=InPosePreFullBodyAimOffsets}
   Output Pose() {layerGroup=Item Override}
  ]}
 ItemFullBodyOverride {nodes [
   Input Pose() {name=InPoseFullBody}
   Output Pose() {layerGroup=Item Override}
  ]}
 ItemIdleOverride {nodes [Output Pose(Result=Input Pose() {name=InPose_Idle}) {layerGroup=Item Override}]}
 ItemJetPackJumpAdditive {nodes [Output Pose() {layerGroup=Item Override}]}
 ItemJetPackStartAdditive {nodes [Output Pose() {layerGroup=Item Override}]}
 ItemJumpLoopAdditive {nodes [
   Input Pose() {name=InPoseItemJumpLoopAdditive}
   Output Pose() {layerGroup=Item Override}
  ]}
 ItemJumpOverride {nodes [
   Input Pose() {name=InPoseJump}
   Output Pose() {layerGroup=Item Override}
  ]}
 ItemJumpUpAdditive {nodes [
   Input Pose() {name=InPoseItemJumpUpAdditive}
   Output Pose() {layerGroup=Item Override}
  ]}
 ItemPreAimOffsets {nodes [
   Input Pose() {name=InPosePreAimOffsets}
   Output Pose() {layerGroup=Item Override}
  ]}
 ItemPreIK {nodes [
   Input Pose() {name=InPose_PreIK}
   Output Pose() {layerGroup=Item Override}
  ]}
 ItemRightHandIK {nodes [
   Input Pose() {name=InPosePreRightHandIK}
   Input Pose() {name=InPosePostRightHandIK}
   Output Pose() {layerGroup=Item Override}
  ]}
 ItemSkeletalControl {nodes [
   Input Pose()
   Output Pose() {layerGroup=Item Override}
  ]}
 ItemSwimJumpFallLoop {nodes [Output Pose() {layerGroup=Item Override}]}
 ItemSwimJumpSurfaceEnd {nodes [Output Pose() {layerGroup=Item Override}]}
 ItemSwimJumpSurfaceLoop {nodes [Output Pose() {layerGroup=Item Override}]}
 ItemUpperBody {nodes [
   Input Pose!() -> (Pose, InputParam:Fort Anim Input Adjusted Aim Structure) {name=InPoseUpperBody}
   Output Pose() {layerGroup=Item Override}
  ]}
 ItemZipLineStartAdditive {nodes [Output Pose() {layerGroup=Item Override}]}
 UpdateMotionMatchingNode_PreNativize {nodes [
   n1 = UpdateMotionMatchingNode_PreNativize!() -> (then, Context:"Anim Update Context Structure (by ref)", Node:"Anim Node Reference Structure (by ref)") {is=title K2Node_FunctionEntry}
   > n2 = Evaluate Proxy: MotionMatchingLODProxyAsset!(
    execute:Exec
    AnimInstance:Anim Instance Object Reference
    ProxyTable:Proxy Table Object Reference=Break Motion Matching LODConfig(MotionMatchingLODConfig=Get CurrentLODConfig())
   ) -> (execute:Exec, Result:Array of Pose Search Database Object References) {is=title K2Node_EvaluateProxy2, mode=AllResults, proxy=/Game/Characters/Player/ProxyAssets/MotionMatchingLODProxyAsset}
   > SetDatabasesToSearch!(
    execute
    MotionMatchingNode:"Motion Matching Anim Node Reference Structure (by ref)"=Convert to Motion Matching Node!(Node:"Anim Node Reference Structure (by ref)"=@n1.Node) -> (MotionMatchingNode:Motion Matching Anim Node Reference Structure, Result:bool)
    Databases:Array of Pose Search Database Object References=@n2
    InterruptMode:EPoseSearchInterruptMode Enum=Get DatabaseSearchInterruptMode() -> (DatabaseSearchInterruptMode:EPoseSearchInterruptMode Enum=DoNotInterrupt)
   ) -> (then)
   Return Node()
  ]}
]
variables [
 ArmLeftLocomotionWeight {type=number}
 ArmRightLocomotionWeight {type=number}
 IsCrouchTransition {type=boolean}
 LeftFingersWepAlpha {type=number}
 OwnerCharMoveComp {type=CharacterMovementComponent, default=null}
 RightFingersWepAlpha {type=number}
 RunLeanAlpha {type=number, default=0.5}
 TorsoLocomotionWeight {type=number}
]
