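# Editor utility blueprint that strips notify tracks containing sync markers from the animation sequences selected in the editor.
# Its one function loops over selected assets casting to AnimSequence, then checks each notify track for sync markers and removes any track that has them.
# If no tracks remain, it adds a default white track named "1", then saves the sequence if dirty, logging progress along the way.
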EditorUtilityBlueprint /LocomotionInitiative/Blueprints/LocomotionInitiativeEditorUtilityBP
graphs [
 EventGraph {links [], nodes []}
 Remove Any Tracks With Sync Markers {nodes [
   Remove Any Tracks With Sync Markers() {is=title K2Node_FunctionEntry}
   > LogString(InString=Editor Utility | Animation Assets | Remove Any Tracks With Sync Markers: Started.)
   > n1 = GetSelectedAssets()
   > n2 = For Each Loop(Array:Array of Object References=@n1) -> (Array Element:Object Reference)
   LogString(execute=@n2.Completed, InString=Editor Utility | Animation Assets | Remove Any Tracks With Sync Markers: Finished.)
   n3 = Cast To AnimSequence!(execute=@n2.LoopBody, Object=@n2) -> (then, CastFailed, AsAnim Sequence:Anim Sequence Object Reference)
   > Set AnimSequenceToBeProcessed(AnimSequenceToBeProcessed:Anim Sequence Object Reference=@n3) -> (Output_Get:Anim Sequence Object Reference)
   > n4 = For Each Loop(Array:Array of Names=GetAnimationNotifyTrackNames!(AnimationSequenceBase:Anim Sequence Base Object Reference=Get AnimSequenceToBeProcessed() -> (AnimSequenceToBeProcessed:Anim Sequence Object Reference)) -> (TrackNames:Array of Names)) -> (Array Element:Name)
   n5 = Get AnimSequenceToBeProcessed()
   > Branch(Condition=Is Not Empty!(TargetArray:Array of Anim Sync Marker Structures=GetAnimationSyncMarkersForTrack!(AnimationSequence:Anim Sequence Object Reference=@n5, NotifyTrackName:Name=@n4) -> (Markers:Array of Anim Sync Marker Structures)) -> (ReturnValue:bool))
   > RemoveAnimationNotifyTrack!(execute, AnimationSequenceBase:Anim Sequence Base Object Reference=Get AnimSequenceToBeProcessed(), NotifyTrackName:Name=@n4) -> (then)
   n6 = Get AnimSequenceToBeProcessed()
   > LogString(InString=Append(A="Removed Notify Track with any Sync Markers for: ", B=GetDisplayName(Object=@n6)))
   n7 = Get AnimSequenceToBeProcessed()
   n8 = Branch(
    execute=@n4.Completed
    Condition=Is Empty!(TargetArray:Array of Names=GetAnimationNotifyTrackNames(AnimationSequenceBase=@n7)) -> (ReturnValue:bool)
   )
   > n9 = AddAnimationNotifyTrack!(
    execute
    AnimationSequenceBase:Anim Sequence Base Object Reference=Get AnimSequenceToBeProcessed()
    NotifyTrackName:Name=1
    TrackColor:Linear Color Structure=(R=1,G=1,B=1,A=1)
   ) -> (then)
   SaveLoadedAsset(execute=[@n9.then, @n8.else], AssetToSave=Get AnimSequenceToBeProcessed(), bOnlyIfIsDirty=true)
  ]}
]
variables []
